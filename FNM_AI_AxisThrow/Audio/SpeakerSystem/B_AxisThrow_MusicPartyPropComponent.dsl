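; B_AxisThrow_MusicPartyPropComponent is a component that auto-starts the music party on the AxisThrow speaker prop that owns it.
; SetupAutoActivate runs only when auto activate is on and music is not bypassed.
; It casts the owner to B_AxisThrow_Speaker, gets that speaker's SpookyAudioSpeaker, and calls StartParty on it.
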(fn SetupAutoActivate
  (if (and (Variables|Activation|GetbAutoActivate) (not (Variables|Assets|GetBypassMusic)))
    (bind _asb_axis_throw_speaker (Utilities|Casting|CastToB_AxisThrow_Speaker (Components|GetOwner)))
    (bind _spookyaudiospeaker (Variables|/FNM_AI_AxisThrow/Audio/SpeakerSystem/B_AxisThrow_Speaker.B_AxisThrow_Speaker_C|GetSpookyAudioSpeaker _asb_axis_throw_speaker)
      (:then
        (CallFunction|StartParty _spookyaudiospeaker))
      (:else))))
