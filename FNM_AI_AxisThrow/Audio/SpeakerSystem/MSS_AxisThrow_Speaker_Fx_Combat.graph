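# Looping stereo MetaSound source for AxisThrow speakers, replaying the combat speaker audio bus with band-pass colouring and speaker attenuation.
# The Default page band-passes each channel, mixes it with a low-passed stereo delay of itself, then applies a fixed gain.
# The Mobile page drops the delay and mixer, keeping only band-pass and gain.
# The Cutoff Frequency input is unconnected, and base submix output is disabled in favour of submix sends.
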MetaSoundSource /FNM_AI_AxisThrow/Audio/SpeakerSystem/MSS_AxisThrow_Speaker_Fx_Combat
inputs [
 AudioBus {type=AudioBusAsset}
 Cutoff Frequency {type=Float}
 UE.Source.OnPlay {type=Trigger}
]
interfaces [UE.OutputFormat.Stereo:1.0, UE.Source:1.1]
outputs [
 "UE.OutputFormat.Stereo.Audio:0" {type=Audio}
 "UE.OutputFormat.Stereo.Audio:1" {type=Audio}
]
pages [
 Default {
  is_build_graph=true
  input_defaults {AudioBus=&/AB_AxisThrow_CombatSpeaker, Cutoff Frequency=579.9993, UE.Source.OnPlay=null}
  nodes [
   AudioBusReader2_0 = "UE.Audio Bus Reader (2)"(Audio Bus=^AudioBus)
   StateVariableFilter_0 = UE.State Variable Filter:Audio(In=@AudioBusReader2_0.Out 0) {also {defaults {Cutoff Frequency=580, Resonance=1}}}
   StateVariableFilter_1 = UE.State Variable Filter:Audio(In=@AudioBusReader2_0.Out 1) {also {defaults {Cutoff Frequency=580, Resonance=1}}}
   StereoDelay_0 = UE.Stereo Delay:Audio(In Left=@StateVariableFilter_0.Band Pass, In Right=@StateVariableFilter_1.Band Pass) {also {defaults {Delay Mode=0, Delay Ratio=0.864, Delay Time=0.12, Feedback=0.104, Wet Level=0.376}}}
   OnePoleLowPassFilter_0 = UE.One-Pole Low Pass Filter:Audio(In=@StereoDelay_0.Out Right) {also {defaults {Cutoff Frequency=2000}}}
   OnePoleLowPassFilter_1 = UE.One-Pole Low Pass Filter:Audio(In=@StereoDelay_0.Out Left) {also {defaults {Cutoff Frequency=2000}}}
   AudioMixerAudioMixerStereo2_0 = "AudioMixer.Audio Mixer (Stereo, 2)"(
    In 0 R=@OnePoleLowPassFilter_0.Out
    In 0 L=@OnePoleLowPassFilter_1.Out
    In 1 L=UE.Reroute(Value=@StateVariableFilter_0.Band Pass).Value
    In 1 R=UE.Reroute(Value=@StateVariableFilter_1.Band Pass).Value
   ) {also {defaults {Gain 0=0.376}}}
   DecibelstoLinearGain_0 = UE.Decibels to Linear Gain:Float() {also {defaults {Decibels=2}}}
   Multiply_0 = UE.Multiply:Audio by Float(PrimaryOperand=@AudioMixerAudioMixerStereo2_0.Out R, AdditionalOperands=@DecibelstoLinearGain_0.Linear Gain)
   Multiply_1 = UE.Multiply:Audio by Float(PrimaryOperand=@AudioMixerAudioMixerStereo2_0.Out L, AdditionalOperands=@DecibelstoLinearGain_0.Linear Gain)
  ]
  links [{dst="@UE.OutputFormat.Stereo.Audio:1", src=Multiply_0.Out}, {dst="@UE.OutputFormat.Stereo.Audio:0", src=Multiply_1.Out}]
 }
 Mobile {
  nodes [
   DecibelstoLinearGain_0 = UE.Decibels to Linear Gain:Float() {also {defaults {Decibels=2}}}
   AudioBusReader2_0 = "UE.Audio Bus Reader (2)"(Audio Bus=^AudioBus)
   StateVariableFilter_1 = UE.State Variable Filter:Audio(In=@AudioBusReader2_0.Out 1) {also {defaults {Cutoff Frequency=580, Resonance=1}}}
   Multiply_0 = UE.Multiply:Audio by Float(AdditionalOperands=@DecibelstoLinearGain_0.Linear Gain, PrimaryOperand=@StateVariableFilter_1.Band Pass)
   StateVariableFilter_0 = UE.State Variable Filter:Audio(In=@AudioBusReader2_0.Out 0) {also {defaults {Cutoff Frequency=580, Resonance=1}}}
   Multiply_1 = UE.Multiply:Audio by Float(AdditionalOperands=@DecibelstoLinearGain_0.Linear Gain, PrimaryOperand=@StateVariableFilter_0.Band Pass)
  ]
  links [{dst="@UE.OutputFormat.Stereo.Audio:1", src=Multiply_0.Out}, {dst="@UE.OutputFormat.Stereo.Audio:0", src=Multiply_1.Out}]
 }
]
properties {
 assetImportData=&self:AssetImportData
 attenuationSettings=&/ATTN_AxisThrow_Speaker
 bDebug=true
 bEnableBaseSubmix=false
 bEnableBusSends=false
 bLooping=true
 duration=10000
 maxDistance=5000
 numChannels=2
 outputFormat=Stereo
 sampleRate=48000
 soundClassObject=/Game/Sounds/FortSoundClassesAndModes/Fort_Music_POI_PSM
 virtualizationMode=PlayWhenSilent
 volume=2
 assetUserData [&self:AudioAssetUserData_0]
 soundSubmixSends [{}, {soundSubmix=/Game/Sounds/SubMixes/Fortnite_LicensedContent_3D}]
}
